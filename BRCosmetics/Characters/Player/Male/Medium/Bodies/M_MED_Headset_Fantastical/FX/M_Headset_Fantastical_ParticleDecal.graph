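# Additive unlit Niagara sprite material for the Fantastical Headset cosmetic that fakes a decal wave on scene surfaces behind the sprite.
# A world-space sphere mask, sized by particle Progress or a sine debug fallback, shapes a noise-distorted radial texture pattern.
# A curve-atlas gradient, particle color and normal-versus-camera shading tint it.
# Scene/custom depth tests, sprite edge and depth fades limit opacity, with normal and depth tests bypassed on mobile.
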Material /BRCosmetics/Characters/Player/Male/Medium/Bodies/M_MED_Headset_Fantastical/FX/M_Headset_Fantastical_ParticleDecal
graph {
 properties {
  bUsedWithInstancedStaticMeshes=true
  bUsedWithNiagaraSprites=true
  blendMode=BLEND_Additive
  pixelDepthOffsetMode=PDOM_Legacy
  shadingModel=MSM_Unlit
  usedShadingModels=Unlit
 }
 nodes [
  n1 = ParticleColor()
  n2 = DotProduct(A=ComponentMask(None=SceneTexture() {sceneTextureId=PPI_WorldNormal}.Color) {b=true, g=true, r=true}, B=CameraVectorWS())
  n3 = ShadingPathSwitch(Default=@n2, Deferred=@n2, Mobile=Constant() {r=1}) {desc=Migrated from FeatureLevelSwitch}
  n4 = TextureObjectParameter() {
   group=ParticleDecal
   parameterName=ParticleDecalPartternNoise_Texture
   sortPriority=0
   texture=/Game/Effects/Fort_Effects/Emotes/UltraLight/Texture/T_SoftNoise_BC7
  }
  n5 = VectorParameter() {
   group=ParticleDecal
   parameterName=ParticleDecalPartternNoise_UVControl
   channelNames {a=VPan, b=UPan, g=VScale, r=UScale}
   defaultValue=<rgba 5 0.04 -0.1 -50>
  }
  n6 = Constant3Vector() {constant=<rgba 0 0 1 0>}
  n7 = Constant3Vector() {constant=<rgba 1 0 0 0>}
  n8 = MaterialFunctionCall(Vector or UVs=MaterialFunctionCall(
    Plane Normal Axis=Transform(None=@n6) {transformSourceType=TRANSFORMSOURCE_Camera}
    Plane X Axis=Transform(None=@n7) {transformSourceType=TRANSFORMSOURCE_Camera}
    Plane Center=ParticlePositionWS()
   ) {materialFunction=&VirtualPlaneCoordinates}.UVs) {materialFunction=&VectorToRadialValue}
  n9 = DynamicParameter() {paramNames [Noise U Offset, "Noise  V Offset", Param3, Progress]}
  n10 = VectorParameter() {
   group=ParticleDecal
   parameterName=ParticleDecalPartternNoise_NoiseControl
   sortPriority=60
   channelNames {a=VStrength, b=UStrength, g=VOffset, r=UOffset}
   defaultValue=<rgba 0 0 0.05 0.1>
  }
  n11 = MaterialFunctionCall!(
   NoiseValue=MaterialFunctionCall(Texture=@n4, UVControl_RGB=@n5, UVControl_A=@n5, "[Override]UVInput"=@n8.Radial Coordinates) {materialFunction=&FXCommon_TextureSample}.RG
   Control_RGB=MaterialFunctionCall(R=@n9.Noise U Offset, G=@n9.Noise U Offset, B=@n10.UStrength) {materialFunction=&AppendMany}.RGB
   Control_A=@n10.VStrength
  ) {materialFunction=/Game/Effects/Niagara/MasterMaterials/Functions/Common/FXCommon_NoiseControl}
  n12 = DynamicParameter() {parameterIndex=1, defaultValue=<rgba 0 1 1 1>, paramNames ["UseParticle(ShouldBe1)", "", Decal U offset, Decal V offset]}
  n13 = MaterialFunctionCall(Float2=@n8.Radial Coordinates) {materialFunction=&BreakOutFloat2Components}
  n14 = ScalarParameter() {defaultValue=4, group=ParticleDecal, parameterName=ParticleDecalParttern_Density_U}
  n15 = ScalarParameter() {defaultValue=0.002, group=ParticleDecal, parameterName=ParticleDecalParttern_Density_V}
  n16 = Add(A=AppendVector(A=@n12.Decal U offset, B=@n12.Decal U offset), B=AppendVector(A=Multiply(A=@n13.R, B=@n14), B=Multiply(A=@n13.G, B=@n15)))
  n17 = TextureSampleParameter2D(UVs=MaterialFunctionCall(
    UVs=Add(A=@n11.NoiseValue, B=@n16)
    "Rotation Angle (0-1)"=ScalarParameter() {group=ParticleDecal, parameterName=ParticleDecalParttern_Rotation}
   ) {materialFunction=&CustomRotator}.Rotated Values) {
   group=ParticleDecal
   parameterName=ParticleDecalParttern_Texture
   samplerType=SAMPLERTYPE_LinearGrayscale
   texture=/Game/Effects/Fort_Effects/Emotes/UltraLight/Texture/T_FluidFlameTile_Greyscale
  }
  n18 = StaticComponentMaskParameter(None=@n17.RGBA) {defaultR=true, group=ParticleDecal, parameterName=ParticleDecalParttern_Channel}
  n19 = VectorParameter() {group=ParticleDecal, parameterName=Decal Refine, sortPriority=4, defaultValue=<rgba 1 1 1 0.5>}
  n20 = DynamicParameter() {parameterIndex=1, defaultValue=<rgba 0 1 1 1>, paramNames ["UseParticle(ShouldBe1)", "", Decal U offset, Decal V offset]}
  n21 = ActorPositionWS()
  n22 = DynamicParameter() {paramNames [Noise U Offset, "Noise  V Offset", Param3, Progress]}
  n23 = Multiply(A=Time(), B=ScalarParameter() {defaultValue=1, group=ParticleDecal, parameterName=ParticleDecalMask_DebugProgressTimeScale})
  n24 = Multiply(A=Add(A=Sine(None=@n23))) {constB=0.5}
  n25 = SphereMask(
   A=MaterialFunctionCall() {materialFunction=&WorldPositionBehindTranslucency}.Result
   B=If(A=@n20."UseParticle(ShouldBe1)", "A > B"=ParticlePositionWS(), "A == B"=@n21, "A < B"=@n21)
   Radius=LinearInterpolate(
    B=ScalarParameter() {defaultValue=70, group=ParticleDecal, parameterName=ParticleDecalMask_Radius}
    Alpha=If(A=@n20."UseParticle(ShouldBe1)", "A > B"=@n22.Progress, "A == B"=@n24, "A < B"=@n24)
   )
   Hardness=ScalarParameter() {group=ParticleDecal, parameterName=ParticleDecalMask_Hardness}
  )
  n26 = OneMinus(None=@n25)
  n27 = ScalarParameter() {group=ParticleDecal, parameterName=ParticleDecalMask_Invert}
  n28 = Add(
   A=Subtract(A=MaterialFunctionCall(Texture=@n18, RGB=@n19.RGB, A=@n19.RGB) {materialFunction=&MF_FX_Refine}.Result)
   B=Multiply(A=LinearInterpolate(A=@n25, B=@n26, Alpha=@n27)) {constB=2}
  )
  n29 = CurveAtlasRowParameter(CurveTime=SphereMask(A=TextureCoordinate(), B=Constant2Vector() {g=0.5, r=0.5}) {attenuationRadius=0.5, hardnessPercent=0}) {
   atlas=/Game/Athena/LinearColorCurves/A_Atlas_Turtleneck_Glider
   curve=/Game/Characters/Player/Male/Medium/Bodies/M_MED_Headset_Fantastical/FX/Curve_Headset_Fantastical_Wave
   defaultValue=37
  }
  n30 = Desaturation(None=Multiply(A=@n28, B=@n29), Fraction=ScalarParameter() {parameterName=Desaturation})
  MaterialFunctionCall(Emissive=Multiply(A=@n1.RGB, B=Multiply(A=@n3, B=@n30))) {materialFunction=&MF_FX_EmissiveAdjustments}
  n31 = SphereMask(
   A=TextureCoordinate()
   B=Constant() {r=0.5}
   Radius=ScalarParameter() {defaultValue=0.5, group=ParticleDecal, parameterName=SpriteEdgeFade_Radius}
   Hardness=ScalarParameter() {defaultValue=0.5, group=ParticleDecal, parameterName=SpriteEdgeFade_Hardness}
  )
  n32 = Constant() {r=1}
  n33 = SceneDepth()
  n34 = ComponentMask(None=SceneTexture() {sceneTextureId=PPI_CustomDepth}.Color) {r=true}
  n35 = Subtract(A=StaticSwitchParameter(True=@n33.R, False=@n34) {group=ParticleDecal, parameterName=UseSceneDepth?}, B=PixelDepth().R)
  n36 = Constant() {r=1}
  n37 = Constant()
  n38 = Constant() {r=1}
  n39 = Multiply(
   A=If(
    A=MaterialFunctionCall(Float3=@n35) {materialFunction=&BreakOutFloat3Components}.R
    B=ScalarParameter() {defaultValue=1000, group=ParticleDecal, parameterName=ParticleDecalMask_MaxDepth}
    "A > B"=Constant()
    "A == B"=@n36
    "A < B"=@n36
   )
   B=StaticSwitchParameter(True=If(A=@n34, B=@n33.R, "A > B"=@n37, "A == B"=@n38, "A < B"=@n38), False=@n32) {defaultValue=true, group=ParticleDecal, parameterName=UseSceneDepthCompare?}
  )
  n40 = ScalarParameter() {defaultValue=1, group=ParticleDecal, parameterName=ParticleDecalParttern_Hardness}
  n41 = Multiply(
   A=ShadingPathSwitch(Default=@n39, Deferred=@n39, Mobile=@n32) {desc=Migrated from FeatureLevelSwitch}
   B=Saturate(None=Multiply(A=@n28, B=@n40))
  )
  Multiply(A=DepthFade(Opacity=Multiply(A=@n31, B=@n41), FadeDistance=ScalarParameter() {defaultValue=50, parameterName=FadeDistance}), B=@n1.A)
 ]
}
