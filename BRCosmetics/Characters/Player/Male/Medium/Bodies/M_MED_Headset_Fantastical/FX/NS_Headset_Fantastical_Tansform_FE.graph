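# Niagara transformation effect for the Fantastical Headset cosmetic body, apparently the front-end (FE) variant.
# Particles attach to the character's skeletal mesh: violet flame sprites at head and hand sockets, curl-noise sparks, and mirrored eye-trail meshes fed a smoothed position.
# User.FX Stage comparisons multiply flame, spark and eye-trail burst counts by one or zero.
# Three ungated wave sprites add bright, fading pulses animated through dynamic material parameters.
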NiagaraSystem /BRCosmetics/Characters/Player/Male/Medium/Bodies/M_MED_Headset_Fantastical/FX/NS_Headset_Fantastical_Tansform_FE
dependencies {
 usedDataInterfaces [
  &NiagaraDataInterfaceCurve
  &NiagaraDataInterfaceColorCurve
  &NiagaraDataInterfaceArrayFloat3
  &NiagaraDataInterfaceSkeletalMesh
 ]
 usedDynamicInputs []
 usedModules [
  &self:SystemSpawnScript.SystemScriptSource.SystemScriptGraph.NiagaraNodeAssignment_0.SetVariables_#1
  &SystemState
 ]
}
emitters {
 Eyes_Trail {
  emitterUpdateScript [
   &EmitterState(Life Cycle Mode=::NewEnumerator1, Loop Behavior=::NewEnumerator1, Loop Duration=1, UseLoopDelay=false)
   &SpawnBurst_Instantaneous(
    Spawn Count=&Multiply_Int(
     A=2
     B=&MakeIntFromFloat(Float=&MakeCustomFloatFromBool(
       Bool=&SetBoolByFloatComparison(A=&MakeFloatFromInt(Int=$User.FX Stage), B=4)
       True Float=1
       False Float=0
      ))
    )
    Use Spawn Probability=true
    UseLoopCountLimit=false
   )
  ]
  particleSpawnScript [
   &FN_SetSpawnGroup(Number of Spawn Groups=2)
   &InitializeParticle(
    Lifetime=0.8
    Mesh Scale=&SelectVectorFromArray(Direct Array Index=$Particles.SpawnGroup, Vector Selection Array=DataInterface() {propertyValues {FloatData [<1.5 1 1.5>, <1.5 -1 1.5>]}})
    Mesh Scale Mode=::NewEnumerator4
    UsePositionOffset=false
    Write Material Random=true
   )
   &InitialMeshOrientation(
    Mesh Orientation Mode=::NewEnumerator2
    Rotation=&SelectVectorFromArray(Direct Array Index=$Particles.SpawnGroup, Vector Selection Array=DataInterface() {propertyValues {FloatData [<0 0 0>, <0 0 0>]}})
    Use Rotation Vector=true
   )
   &CameraOffset(Camera Offset Amount=0.5)
  ]
  particleUpdateScript [
   &ParticleState()
   &SkeletalMeshLocation(
    MaskSpawnGroup=false
    Mesh Sampling Type=::NewEnumerator5
    Sampled Position Offset=&SelectVectorFromArray(Direct Array Index=$Particles.SpawnGroup, Vector Selection Array=DataInterface() {propertyValues {FloatData [<0 -2 0>, <0 2 0>]}})
    Skeletal Mesh=$System.Skeletal Mesh
    Socket Sampling Mode=::NewEnumerator5
   )
   &InterpolateOverTime(
    Rate Of Change=8
    Target Vector Value:Vector3f=&ConvertPositionToVector(Input Position=$Particles.Position, Convert To Absolute Large World Vector=false)
    Variable Type=::NewEnumerator2
   )
   /SetVariables(Particles.Interpolate Dynamic=&MakeFloat4FromVectorAndFloat("Vector (RGB)"=$StackContext.InterpolateOverTime.InterpolateOverTime_V3_.Moving Average, "Float (Alpha)"=0.3)) {isSetParameters=true}
   &ScaleColor(
    Scale Alpha=&FloatFromCurve(FloatCurve=DataInterface() {propertyValues {Curve {keys [{interpMode=RCIM_Cubic, tangentMode=RCTM_User, time=0.6, value=1}, {interpMode=RCIM_Cubic, tangentMode=RCTM_User, time=1, value=0}]}}})
    Scale RGB=&Multiply_VectorByFloat(
     Vector=<1 1 1>
     Float=&FloatFromCurve(FloatCurve=DataInterface() {propertyValues {Curve {keys [
          {arriveTangent=-19.9892, interpMode=RCIM_Cubic, leaveTangent=-19.9892, tangentMode=RCTM_User, time=0, value=20}
          {interpMode=RCIM_Cubic, tangentMode=RCTM_User, time=1, value=2}
         ]}}})
    )
    ScaleA=true
    ScaleRGB=true
   )
   &DynamicMaterialParameters(
    Index 0 Param 1=&RandomRangeFloat(Minimum=0, Maximum=1)
    Index 0 Param 3=1
    Index 0 Param 4=100
    Index 3 Param 1=&FloatFromCurve(FloatCurve=DataInterface() {propertyValues {Curve {keys [
         {arriveTangent=1.7778, interpMode=RCIM_Cubic, leaveTangent=1.7778, tangentMode=RCTM_User, time=0, value=0}
         {arriveTangent=0.1208, interpMode=RCIM_Cubic, leaveTangent=0.1208, tangentMode=RCTM_User, time=1, value=1}
        ]}}})
    Index 3 Param 2=1
    Module.Param0WriteEnabled=true
    Module.Param0WriteEnabled3=true
    Module.Param1WriteEnabled=true
    Module.Param1WriteEnabled3=true
    Module.Param2WriteEnabled=true
    Module.Param2WriteEnabled3=true
    Module.Param3WriteEnabled=true
    Module.Param3WriteEnabled3=true
    Write Parameter Index 3=true
   )
  ]
  rendererClasses [&NiagaraMeshRendererProperties]
  renderers [
   {rendererClass=&NiagaraMeshRendererProperties, rendererIndex=0}
   {rendererClass=&NiagaraMeshRendererProperties, rendererIndex=1}
  ]
 }
 Flame_Hand {
  emitterUpdateScript [
   &EmitterState(Life Cycle Mode=::NewEnumerator1, Loop Behavior=::NewEnumerator1, Loop Delay=0.1, Loop Duration=1, UseLoopDelay=true)
   &SpawnBurst_Instantaneous(
    Spawn Count=&Multiply_Int(
     A=3
     B=&MakeIntFromFloat(Float=&MakeCustomFloatFromBool(
       Bool=&SetBoolByFloatComparison(A=&MakeFloatFromInt(Int=$User.FX Stage), B=2)
       True Float=1
       False Float=0
      ))
    )
    Use Spawn Probability=false
    UseLoopCountLimit=false
   )
  ]
  particleSpawnScript [
   /Niagara/Modules/Spawn/Initialization/InitializeParticle(
    Color=<rgba 2 2 2 1>
    Lifetime=&UniformRangedFloat(Minimum=0.55, Maximum=0.3)
    Module.Write Color=true
    Module.Write Lifetime=true
    Module.Write Mass=false
    Module.Write Material Random=true
    Module.Write Position=false
    Module.Write Scale=false
    Module.Write SpriteRotation=true
    Module.Write SpriteSize=true
    Position=&SimulationPosition()
    Sprite Rotation=&UniformRangedFloat(Minimum=0, Maximum=360)
    Sprite Size=&Vector2DFromFloat(Value=&RandomRangeFloat(Minimum=15, Maximum=30))
   )
   &SphereLocation(
    MaskSpawnGroup=false
    Offset=<0 -8 -2>
    Override Seed=false
    Sphere Radius=10
    Surface Only=false
    UseNonUniformScale=false
    UseOffset=true
    UseSphereOrientation=false
   )
   &DynamicMaterialParameters(
    Index 0 Param 2=&UniformRangedFloat(Minimum=0, Maximum=1)
    Index 0 Param 3=1
    Index 0 Param 4=&UniformRangedFloat(Minimum=1, Maximum=2)
    Module.Param0WriteEnabled=false
    Module.Param1WriteEnabled=true
    Module.Param2WriteEnabled=false
    Module.Param3WriteEnabled=true
   ) {module=DynamicMaterialParameters002}
   &CameraOffset(Camera Offset Amount=&RandomRangeFloat(Minimum=15, Maximum=10))
   &AddVelocity(
    Velocity=&RandomRangeVector(Minimum=<0.5 0.5 1>, Maximum=<-0.5 -0.5 -1>)
    Velocity Speed Scale=&RandomRangeFloat(Minimum=20, Maximum=50)
   )
   &ApplyInitialForces(Apply Rotational Force to Rotational Velocity=false)
   &SkeletalMeshLocation(
    Bone / Socket Index=&RandomRangeInt(Minimum=2, Maximum=4)
    MaskSpawnGroup=false
    Mesh Sampling Type=::NewEnumerator5
    Sampled Position Offset=&RandomRangeVector(Minimum=<0 0 0>, Maximum=<-5 0 0>)
    Skeletal Mesh=$System.Skeletal Mesh
    Socket Sampling Mode=::NewEnumerator5
   )
  ]
  particleUpdateScript [
   &ParticleState()
   &ScaleSpriteSize(Scale Factor=&Vector2DFromFloat(Value=&FloatFromCurve(FloatCurve=DataInterface() {propertyValues {Curve {keys [
          {arriveTangent=1.9543, interpMode=RCIM_Cubic, leaveTangent=1.9543, tangentMode=RCTM_User, time=0, value=0.1}
          {arriveTangent=0.1045, interpMode=RCIM_Cubic, leaveTangent=0.1045, tangentMode=RCTM_User, time=0.3, value=0.85}
          {arriveTangent=-0.8944, interpMode=RCIM_Cubic, leaveTangent=-0.8944, tangentMode=RCTM_User, time=1, value=0.8}
         ]}}})))
   /Niagara/Modules/Update/SubUV/SubUVAnimation(Number Of Frames=&FloatFromCurve(FloatCurve=DataInterface() {propertyValues {Curve {keys [<at 0 0>,<at 1 40>]}}}))
   &DynamicMaterialParameters(
    Index 0 Param 1=&FloatFromCurve(FloatCurve=DataInterface() {propertyValues {Curve {keys [{arriveTangent=2.7551, interpMode=RCIM_Cubic, leaveTangent=2.7551, tangentMode=RCTM_User, time=0, value=0}, <at 0.7 1>]}}})
    Index 0 Param 2=1
    Index 0 Param 3=1
    Index 0 Param 4=1
    Module.Param0WriteEnabled=true
    Module.Param1WriteEnabled=true
    Module.Param2WriteEnabled=true
    Module.Param3WriteEnabled=false
   ) {module=DynamicMaterialParameters001}
   &GravityForce(Gravity=&RandomRangeVector(Minimum=<0 0 300>, Maximum=<0 0 100>))
   &Drag(Drag=4, Use Linear Drag=true, Use Rotational Drag=false)
   &SolveForcesAndVelocity(Clamp Velocity=false, Limit Acceleration=false)
   &ScaleColor(
    Scale Mode=::NewEnumerator1
    Scale RGBA=&Vector4FromCurve(Vector4Curve=DataInterface() {propertyValues {
       AlphaCurve {keys [<at 0.0024 0>,<at 0.2857 1>,<at 0.6079 1>,<at 1 0>]}
       BlueCurve {keys [<at 0 5>,<at 0.1945 3>,<at 0.7802 1.2674>,<at 0.8926 0.5>]}
       GreenCurve {keys [<at 0 0.25>,<at 0.1945 0.15>,<at 0.7802 0.0634>,<at 0.8926 0>]}
       RedCurve {keys [<at 0 1.0417>,<at 0.1945 0.625>,<at 0.7802 1.2674>,<at 0.8926 0.3333>]}
      }})
    ScaleRGBA=true
   )
  ]
  rendererClasses [&NiagaraSpriteRendererProperties]
  renderers [{rendererClass=&NiagaraSpriteRendererProperties}]
 }
 Flame_Head {
  emitterUpdateScript [
   &EmitterState(Life Cycle Mode=::NewEnumerator1, Loop Behavior=::NewEnumerator1, Loop Delay=0.05, Loop Duration=1, UseLoopDelay=true)
   &SpawnBurst_Instantaneous(
    Spawn Count=&Multiply_Int(
     A=3
     B=&MakeIntFromFloat(Float=&MakeCustomFloatFromBool(
       Bool=&SetBoolByFloatComparison(A=&MakeFloatFromInt(Int=$User.FX Stage), B=2)
       True Float=1
       False Float=0
      ))
    )
    Use Spawn Probability=false
    UseLoopCountLimit=false
   )
  ]
  particleSpawnScript [
   &InitializeParticle(
    Color=<rgba 2 2 2 1>
    Lifetime=&UniformRangedFloat(Minimum=0.55, Maximum=0.3)
    Module.Write Color=true
    Module.Write Lifetime=true
    Module.Write Mass=false
    Module.Write Material Random=true
    Module.Write Position=false
    Module.Write Scale=false
    Module.Write SpriteRotation=true
    Module.Write SpriteSize=true
    Position=&SimulationPosition()
    Sprite Rotation=&UniformRangedFloat(Minimum=0, Maximum=360)
    Sprite Size=&Vector2DFromFloat(Value=&RandomRangeFloat(Minimum=20, Maximum=40))
   )
   &SphereLocation(
    MaskSpawnGroup=false
    Offset=<0 -8 -2>
    Override Seed=false
    Sphere Radius=10
    Surface Only=false
    UseNonUniformScale=false
    UseOffset=true
    UseSphereOrientation=false
   )
   &DynamicMaterialParameters(
    Index 0 Param 2=&UniformRangedFloat(Minimum=0, Maximum=1)
    Index 0 Param 3=1
    Index 0 Param 4=&UniformRangedFloat(Minimum=1, Maximum=2)
    Module.Param0WriteEnabled=false
    Module.Param1WriteEnabled=true
    Module.Param2WriteEnabled=false
    Module.Param3WriteEnabled=true
   ) {module=DynamicMaterialParameters002}
   &CameraOffset(Camera Offset Amount=&RandomRangeFloat(Minimum=15, Maximum=10))
   &AddVelocity(
    Velocity=&RandomRangeVector(Minimum=<0.5 0.5 1>, Maximum=<-0.5 -0.5 -1>)
    Velocity Speed Scale=&RandomRangeFloat(Minimum=20, Maximum=50)
   )
   &ApplyInitialForces(Apply Rotational Force to Rotational Velocity=false)
   &SkeletalMeshLocation(
    Bone / Socket Index=&RandomRangeInt(Minimum=0, Maximum=1)
    MaskSpawnGroup=false
    Mesh Sampling Type=::NewEnumerator5
    Sampled Position Offset=&RandomRangeVector(Minimum=<0 0 0>, Maximum=<-5 0 0>)
    Skeletal Mesh=$System.Skeletal Mesh
    Socket Sampling Mode=::NewEnumerator5
   )
  ]
  particleUpdateScript [
   &ParticleState()
   &ScaleSpriteSize(Scale Factor=&Vector2DFromFloat(Value=&FloatFromCurve(FloatCurve=DataInterface() {propertyValues {Curve {keys [
          {arriveTangent=1.9543, interpMode=RCIM_Cubic, leaveTangent=1.9543, tangentMode=RCTM_User, time=0, value=0.1}
          {arriveTangent=0.1045, interpMode=RCIM_Cubic, leaveTangent=0.1045, tangentMode=RCTM_User, time=0.3, value=0.85}
          {arriveTangent=-0.8944, interpMode=RCIM_Cubic, leaveTangent=-0.8944, tangentMode=RCTM_User, time=1, value=0.8}
         ]}}})))
   &SubUVAnimation(Number Of Frames=&FloatFromCurve(FloatCurve=DataInterface() {propertyValues {Curve {keys [<at 0 0>,<at 1 40>]}}}))
   &DynamicMaterialParameters(
    Index 0 Param 1=&FloatFromCurve(FloatCurve=DataInterface() {propertyValues {Curve {keys [{arriveTangent=2.7551, interpMode=RCIM_Cubic, leaveTangent=2.7551, tangentMode=RCTM_User, time=0, value=0}, <at 0.7 1>]}}})
    Index 0 Param 2=1
    Index 0 Param 3=1
    Index 0 Param 4=1
    Module.Param0WriteEnabled=true
    Module.Param1WriteEnabled=true
    Module.Param2WriteEnabled=true
    Module.Param3WriteEnabled=false
   ) {module=DynamicMaterialParameters001}
   &GravityForce(Gravity=&RandomRangeVector(Minimum=<0 0 300>, Maximum=<0 0 50>))
   &Drag(Drag=4, Use Linear Drag=true, Use Rotational Drag=false)
   &SolveForcesAndVelocity(Clamp Velocity=false, Limit Acceleration=false)
   &ScaleColor(
    Scale Mode=::NewEnumerator1
    Scale RGBA=&Vector4FromCurve(Vector4Curve=DataInterface() {propertyValues {
       AlphaCurve {keys [<at 0.0024 0>,<at 0.2857 1>,<at 0.6079 1>,<at 1 0>]}
       BlueCurve {keys [<at 0 5>,<at 0.1945 3>,<at 0.7802 1.2674>,<at 0.8926 0.5>]}
       GreenCurve {keys [<at 0 0.25>,<at 0.1945 0.15>,<at 0.7802 0.0634>,<at 0.8926 0>]}
       RedCurve {keys [<at 0 1.0417>,<at 0.1945 0.625>,<at 0.7802 1.2674>,<at 0.8926 0.3333>]}
      }})
    ScaleRGBA=true
   )
  ]
  rendererClasses [&NiagaraSpriteRendererProperties]
  renderers [{rendererClass=&NiagaraSpriteRendererProperties}]
 }
 Sparks {
  emitterUpdateScript [
   &EmitterState(Life Cycle Mode=::NewEnumerator1, Loop Behavior=::NewEnumerator1, Loop Delay=0.03, Loop Duration=1, UseLoopDelay=true)
   &SpawnBurst_Instantaneous(
    Spawn Count=&Multiply_Int(
     A=&MakeIntFromFloat(Float=&Lerp_Float(
       A=20
       B=70
       Alpha=&MakeCustomFloatFromBool(
        Bool=&SetBoolByFloatComparison(A=&MakeFloatFromInt(Int=$User.FX Stage), B=2)
        True Float=1
        False Float=0
       )
      ))
     B=&MakeIntFromFloat(Float=&MakeCustomFloatFromBool(
       Bool=&SetBoolByFloatComparison(A=&MakeFloatFromInt(Int=$User.FX Stage), B=2)
       True Float=1
       False Float=0
      ))
    )
    Use Spawn Probability=false
    UseLoopCountLimit=false
   )
  ]
  particleSpawnScript [
   /Niagara/Modules/Spawn/Initialization/V2/InitializeParticle(
    Color Maximum=<rgba 3 3 3 1>
    Color Minimum=<rgba 5 5 5 1>
    Color Mode=::NewEnumerator2
    Lifetime Max=0.4
    Lifetime Min=0.2
    Lifetime Mode=::NewEnumerator1
    Sprite Size Max=<1 3>
    Sprite Size Min=<1.5 2>
    Sprite Size Mode=::NewEnumerator2
    UsePositionOffset=false
    Write Material Random=true
   )
   &AddVelocity(
    Velocity=&RandomRangeVector(Minimum=<1 1 1>, Maximum=<-1 -1 -1>)
    Velocity Speed Scale=&RandomRangeFloat(Minimum=20, Maximum=150)
   )
   &InheritVelocity(Speed Limit=true, UseSpeedThreshold=false)
   &ApplyInitialForces()
   &SystemLocation(MaskSpawnGroup=false, Offset=<0 -8 -6>)
   &SkeletalMeshLocation(MaskSpawnGroup=false, Mesh Sampling Type=::NewEnumerator5, Skeletal Mesh=$System.Skeletal Mesh, Socket Sampling Mode=::NewEnumerator5)
  ]
  particleUpdateScript [
   &ParticleState()
   &GravityForce(Gravity=<0 0 100>)
   &Drag(Drag=&RandomRangeFloat(Minimum=2.5, Maximum=4), Use Linear Drag=true, Use Rotational Drag=true)
   &CurlNoiseForce(
    Noise Frequency=200
    Noise Strength=&FloatFromCurve(FloatCurve=DataInterface() {propertyValues {Curve {keys [<at 0 0>,<at 0.2 1.275>,<at 1 1.5>]}}}, Scale Curve=1000)
    Pan Noise=false
   )
   &SolveForcesAndVelocity(Clamp Velocity=false, Limit Acceleration=false)
   &ScaleColor(
    Linear Color Curve=DataInterface() {propertyValues {
      AlphaCurve {keys [<at 0 1>]}
      BlueCurve {keys [<at 0 2>,<at 0.4727 1>,<at 1 1>]}
      GreenCurve {keys [<at 0 0.04>,<at 0.4727 0>,<at 1 0>]}
      RedCurve {keys [<at 0 0.3667>,<at 0.4727 0.1667>,<at 1 1>]}
     }}
    Scale Mode=::NewEnumerator2
   )
   &ScaleSpriteSize(Uniform Curve Sprite Scale=DataInterface() {propertyValues {Curve {keys [
        {interpMode=RCIM_Cubic, tangentMode=RCTM_User, time=0, value=1}
        {arriveTangent=-2.4866, interpMode=RCIM_Cubic, leaveTangent=-2.4866, tangentMode=RCTM_User, time=1, value=0}
       ]}}})
   &ScaleSpriteSizeBySpeed(Max Scale Factor=<0.5 10>, Min Scale Factor=<1 0.5>)
  ]
  rendererClasses [&NiagaraSpriteRendererProperties]
  renderers [{rendererClass=&NiagaraSpriteRendererProperties}]
 }
 Wave_01_FE {
  emitterUpdateScript [
   &EmitterState(
    Inactive Response=::NewEnumerator1
    Life Cycle Mode=::NewEnumerator1
    Loop Behavior=::NewEnumerator1
    Loop Duration Mode=::NewEnumerator1
    UseLoopDelay=false
   )
   &SpawnBurst_Instantaneous(Spawn Count=&Multiply_Int(A=1, B=1), Use Spawn Probability=false, UseLoopCountLimit=false)
  ]
  particleSpawnScript [
   &InitializeParticle(
    Color=<rgba 0.5 0.5 0.5 1>
    Lifetime=0.55
    Sprite Size Mode=::NewEnumerator3
    Uniform Sprite Size=200
    UsePositionOffset=false
    Write Material Random=true
   )
   &CameraOffset(Camera Offset Amount=40)
  ]
  particleUpdateScript [
   &ParticleState()
   &SkeletalMeshLocation(Bone Sampling Mode=::NewEnumerator4, MaskSpawnGroup=false, Skeletal Mesh=$System.Skeletal Mesh)
   &DynamicMaterialParameters(
    Index 0 Param 1=0.5
    Index 0 Param 2=&FloatFromCurve(FloatCurve=DataInterface() {propertyValues {Curve {keys [
         {arriveTangent=0.5735, interpMode=RCIM_Cubic, leaveTangent=0.5735, tangentMode=RCTM_User, time=0, value=0}
         {arriveTangent=-0.1751, interpMode=RCIM_Cubic, leaveTangent=-0.1751, tangentMode=RCTM_User, time=1, value=0.5}
        ]}}})
    Index 0 Param 4=&Multiply_Float(
     A=&FloatFromCurve(FloatCurve=DataInterface() {propertyValues {Curve {keys [
          {arriveTangent=1.7416, interpMode=RCIM_Cubic, leaveTangent=1.7416, tangentMode=RCTM_User, time=0, value=0}
          {arriveTangent=1.8287, interpMode=RCIM_Cubic, leaveTangent=1.8287, tangentMode=RCTM_User, time=0.3, value=1}
          {arriveTangent=0.3567, interpMode=RCIM_Cubic, leaveTangent=0.3567, tangentMode=RCTM_User, time=1, value=2}
         ]}}})
     B=1
    )
    Index 1 Param 1=1
    Module.Param0WriteEnabled=true
    Module.Param0WriteEnabled1=true
    Module.Param1WriteEnabled=true
    Module.Param1WriteEnabled1=true
    Module.Param2WriteEnabled=true
    Module.Param2WriteEnabled1=true
    Module.Param3WriteEnabled=true
    Module.Param3WriteEnabled1=true
    Write Parameter Index 1=true
   )
   &ScaleColor(
    Scale Alpha=&FloatFromCurve(FloatCurve=DataInterface() {propertyValues {Curve {keys [<at 0.3 2>,<at 1 0>]}}})
    Scale RGB=&VectorFromFloat(Value=&FloatFromCurve(
      FloatCurve=DataInterface() {propertyValues {Curve {keys [
          {arriveTangent=-31.5117, interpMode=RCIM_Cubic, leaveTangent=-31.5117, tangentMode=RCTM_User, time=0, value=10}
          {interpMode=RCIM_Cubic, tangentMode=RCTM_User, time=1, value=1}
         ]}}}
      Scale Curve=&Lerp_Float(A=0.8, B=1, Alpha=1)
     ))
    ScaleA=true
    ScaleRGB=true
   )
   &ScaleSpriteSize(Uniform Curve Sprite Scale=DataInterface() {propertyValues {Curve {keys [{tangentMode=RCTM_User, time=0, value=1}, {tangentMode=RCTM_User, time=1, value=2}]}}})
  ]
  rendererClasses [&NiagaraSpriteRendererProperties]
  renderers [{rendererClass=&NiagaraSpriteRendererProperties}]
 }
 Wave_03_FE {
  emitterUpdateScript [
   &EmitterState(
    Inactive Response=::NewEnumerator1
    Life Cycle Mode=::NewEnumerator1
    Loop Behavior=::NewEnumerator1
    Loop Duration Mode=::NewEnumerator1
    UseLoopDelay=false
   )
   &SpawnBurst_Instantaneous(Spawn Count=&Multiply_Int(A=1, B=1), Use Spawn Probability=false, UseLoopCountLimit=false)
  ]
  particleSpawnScript [
   &InitializeParticle(Lifetime=0.5, Sprite Size Mode=::NewEnumerator3, Uniform Sprite Size=100, UsePositionOffset=false, Write Material Random=true)
   &CameraOffset(Camera Offset Amount=50)
  ]
  particleUpdateScript [
   &ParticleState()
   &SkeletalMeshLocation(Bone Sampling Mode=::NewEnumerator4, MaskSpawnGroup=false, Skeletal Mesh=$System.Skeletal Mesh)
   &DynamicMaterialParameters(
    Index 0 Param 4=3
    Index 1 Param 1=1
    Index 1 Param 4=&FloatFromCurve(FloatCurve=DataInterface() {propertyValues {Curve {keys [
         {arriveTangent=-1.3801, interpMode=RCIM_Cubic, leaveTangent=-1.3801, tangentMode=RCTM_User, time=0, value=0.5}
         {arriveTangent=-0.495, interpMode=RCIM_Cubic, leaveTangent=-0.495, tangentMode=RCTM_User, time=1, value=0}
        ]}}})
    Module.Param0WriteEnabled=true
    Module.Param0WriteEnabled1=true
    Module.Param1WriteEnabled=true
    Module.Param1WriteEnabled1=true
    Module.Param2WriteEnabled=true
    Module.Param2WriteEnabled1=true
    Module.Param3WriteEnabled=true
    Module.Param3WriteEnabled1=true
    Write Parameter Index 1=true
   )
   &ScaleColor(
    Scale Alpha=&FloatFromCurve(FloatCurve=DataInterface() {propertyValues {Curve {keys [<at 0 1>,<at 1 0>]}}})
    Scale RGB=&VectorFromFloat(Value=&FloatFromCurve(
      FloatCurve=DataInterface() {propertyValues {Curve {keys [
          {arriveTangent=-26.33, interpMode=RCIM_Cubic, leaveTangent=-26.33, tangentMode=RCTM_User, time=0, value=10}
          {interpMode=RCIM_Cubic, tangentMode=RCTM_User, time=1, value=0}
         ]}}}
      Scale Curve=&Lerp_Float(A=0.5, B=1, Alpha=1)
     ))
    ScaleA=true
    ScaleRGB=true
   )
   &ScaleSpriteSize(Uniform Curve Sprite Scale=DataInterface() {propertyValues {Curve {keys [{tangentMode=RCTM_User, time=0, value=1}, {tangentMode=RCTM_User, time=1, value=1}]}}}) {enabled=false}
  ]
  rendererClasses [&NiagaraSpriteRendererProperties]
  renderers [{rendererClass=&NiagaraSpriteRendererProperties}]
 }
 Wave_04_FE {
  emitterUpdateScript [
   &EmitterState(Life Cycle Mode=::NewEnumerator1, Loop Behavior=::NewEnumerator1, Loop Duration=1, UseLoopDelay=false)
   &SpawnBurst_Instantaneous(Spawn Count=&Multiply_Int(A=1, B=1), Use Spawn Probability=false, UseLoopCountLimit=false)
  ]
  particleSpawnScript [
   &InitializeParticle(Lifetime=0.4, Sprite Size Mode=::NewEnumerator3, Uniform Sprite Size=100, UsePositionOffset=false, Write Material Random=true)
   &CameraOffset(Camera Offset Amount=50)
  ]
  particleUpdateScript [
   &ParticleState()
   &SkeletalMeshLocation(Bone Sampling Mode=::NewEnumerator4, MaskSpawnGroup=false, Skeletal Mesh=$System.Skeletal Mesh)
   &DynamicMaterialParameters(
    Index 0 Param 4=3
    Index 1 Param 1=1
    Index 1 Param 4=&FloatFromCurve(FloatCurve=DataInterface() {propertyValues {Curve {keys [
         {arriveTangent=-0.9127, interpMode=RCIM_Cubic, leaveTangent=-0.9127, tangentMode=RCTM_User, time=0, value=0.4}
         {arriveTangent=-0.495, interpMode=RCIM_Cubic, leaveTangent=-0.495, tangentMode=RCTM_User, time=1, value=-0.3}
        ]}}})
    Module.Param0WriteEnabled=true
    Module.Param0WriteEnabled1=true
    Module.Param1WriteEnabled=true
    Module.Param1WriteEnabled1=true
    Module.Param2WriteEnabled=true
    Module.Param2WriteEnabled1=true
    Module.Param3WriteEnabled=true
    Module.Param3WriteEnabled1=true
    Write Parameter Index 1=true
   )
   &ScaleColor(
    Scale Alpha=&FloatFromCurve(FloatCurve=DataInterface() {propertyValues {Curve {keys [<at 0 1>,<at 1 0>]}}})
    Scale RGB=&VectorFromFloat(Value=&FloatFromCurve(FloatCurve=DataInterface() {propertyValues {Curve {keys [
          {arriveTangent=-26.33, interpMode=RCIM_Cubic, leaveTangent=-26.33, tangentMode=RCTM_User, time=0, value=10}
          {interpMode=RCIM_Cubic, tangentMode=RCTM_User, time=1, value=0}
         ]}}}))
    ScaleA=true
    ScaleRGB=true
   )
  ]
  rendererClasses [&NiagaraSpriteRendererProperties]
  renderers [{rendererClass=&NiagaraSpriteRendererProperties}]
 }
}
summary {
 systemName=NS_Headset_Fantastical_Tansform_FE
 emitters [
  {emitterName=Sparks, rendererClasses [&NiagaraSpriteRendererProperties]}
  {emitterName=Eyes_Trail, rendererClasses [&NiagaraMeshRendererProperties]}
  {emitterName=Flame_Head, rendererClasses [&NiagaraSpriteRendererProperties]}
  {emitterName=Flame_Hand, rendererClasses [&NiagaraSpriteRendererProperties]}
  {emitterName=Wave_01_FE, rendererClasses [&NiagaraSpriteRendererProperties]}
  {emitterName=Wave_03_FE, rendererClasses [&NiagaraSpriteRendererProperties]}
  {emitterName=Wave_04_FE, rendererClasses [&NiagaraSpriteRendererProperties]}
 ]
 userVariables [
  User.bIsFrontEnd {type=NiagaraBool}
  User.bIsFrontEndPreview {type=NiagaraBool}
  User.FX Stage {type=NiagaraInt32, defaultValue=5}
 ]
}
