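# Material function that adds animated-looking flame tattoos to the skin of the Headset Fantastical character cosmetic.
# Distortion noise warps two samples of the SkinFX mask, held steady by its blue channel, to build soft and hard masks.
# Permanent flames and a transition amount drive gradient-colored emissive, dimmed at night via time of day, over base color blended toward near-black.
# It outputs the modified material attributes, flame mask and flame color.
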MaterialFunction /BRCosmetics/Characters/Player/Male/Medium/Bodies/M_MED_Headset_Fantastical/Materials/MF_Headset_MovingTattoos
graph {nodes [
  n1 = TextureCoordinate()
  n2 = TextureSample() {samplerType=SAMPLERTYPE_Masks, texture=/BRCosmetics/Characters/Player/Male/Medium/Bodies/M_MED_Headset_Fantastical/Textures/HeadsetFire_M}
  n3 = MaterialFunctionCall(Float4=FunctionInput(Preview=@n2.RGBA) {inputName=Distortion Noise, inputType=FunctionInput_Vector4}) {materialFunction=&BreakOutFloat4Components}
  n4 = Multiply(A=ScalarParameter() {defaultValue=0.01, parameterName=Tattoo Warp}, B=@n3.R)
  n5 = TextureObject() {
   samplerType=SAMPLERTYPE_Masks
   texture=/BRCosmetics/Characters/Player/Male/Medium/Bodies/M_MED_Headset_Fantastical/Textures/T_M_Headset_Fantastical_Body_FX
  }
  n6 = FunctionInput(Preview=@n5) {bUsePreviewValueAsDefault=true, inputName=SkinFX, inputType=FunctionInput_Texture2D}
  n7 = TextureSample(Tex=@n6) {
   samplerType=SAMPLERTYPE_Masks
   texture=&T_M_Headset_Fantastical_Body_FX
  }
  n8 = TextureSample(UVs=LinearInterpolate(A=Add(A=@n1, B=@n4), B=@n1, Alpha=@n7.B), Tex=@n6)
  n9 = Multiply(A=ScalarParameter() {defaultValue=0.02, parameterName=Tattoo Warp 2}, B=@n3.B)
  n10 = TextureSample(UVs=LinearInterpolate(A=Add(A=@n1, B=@n9), B=@n1, Alpha=@n7.B), Tex=@n6)
  n11 = Add(A=Divide(A=@n8.G), B=Divide(A=@n10.G))
  n12 = ScalarParameter() {defaultValue=0.65, parameterName=Permenant Flames}
  n13 = Multiply(
   A=Multiply(A=Subtract(A=@n11, B=@n12), B=ScalarParameter() {defaultValue=1, parameterName=Perma Flames Intensity})
   B=FunctionInput() {inputName=Character Skin, inputType=FunctionInput_Scalar}
  )
  n14 = Max(A=@n13, B=FunctionInput() {inputName=transition Amount, inputType=FunctionInput_Scalar})
  Add(A=@n13, B=@n14)
  n15 = GetMaterialAttributes(None=FunctionInput() {inputName=Material, inputType=FunctionInput_MaterialAttributes}) {attributeGetTypes [#1, #2]}
  n16 = Constant3Vector() {constant=<rgba 0.02 0.02 0.02 1>}
  n17 = NamedRerouteUsage()
  n18 = NamedRerouteUsage()
  n19 = NamedRerouteUsage()
  n20 = ScalarParameter() {defaultValue=1.05, parameterName=FlameColorBoost}
  n21 = LinearInterpolate(Alpha=Saturate(None=Multiply(A=@n3.R, B=@n20))) {constA=0.1, constB=1.1}
  n22 = TextureSample(UVs=Saturate(None=Multiply(A=@n19, B=@n21))) {texture=/BRCosmetics/Characters/Player/Male/Medium/Bodies/M_MED_Headset_Fantastical/Textures/HeadsetGradient_Grad}
  n23 = Multiply(A=Multiply(A=Multiply(A=@n18) {constB=1.25}, B=@n22.RGB), B=@n14)
  n24 = ScalarParameter() {defaultValue=1.5, parameterName=Flame Emissive Intensity}
  n25 = Add(A=@n15.EmissiveColor, B=Multiply(A=NamedRerouteUsage(), B=Multiply(A=@n23, B=@n24)))
  n26 = Constant() {r=1}
  n27 = MaterialFunctionCall(Value_Morning=@n26, Value_Day=@n26, Value_Evening=@n26, Value_Night=Constant() {r=0.4}) {materialFunction=&Scalar_Time_of_Day}
  FunctionOutput(None=SetMaterialAttributes!(
    MaterialAttributes=@n15.MaterialAttributes
    BaseColor=LinearInterpolate(A=@n15.BaseColor, B=@n16, Alpha=@n17)
    EmissiveColor=Multiply(A=@n25, B=@n27.Result)
   ) {attributeSetTypes [#1, #2]})
  n28 = NamedRerouteUsage()
  FunctionOutput(None=@n28) {outputName=Flame Mask}
  FunctionOutput(None=@n22.RGB) {outputName=Flame Color}
  NamedRerouteDeclaration(None=Power(Base=Max(A=@n8.R, B=@n10.R), Exp=Constant() {r=0.6})) {name=Hard Mask, nodeColor=<rgba 0 1 0.9063 1>}
  NamedRerouteDeclaration(None=@n11) {name=Soft Mask, nodeColor=<rgba 0 0.0156 1 1>}
  NamedRerouteUsage()
  Power(Base=@n28) {constExponent=4}
 ]}
