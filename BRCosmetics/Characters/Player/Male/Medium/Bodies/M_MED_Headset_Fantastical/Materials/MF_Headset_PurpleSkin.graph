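# Material function for the Fantastical Headset character body that shifts surfaces to purple skin with glowing flame edges, driven by a Transition input.
# Desaturated base color is multiplied by Gmap tints split by Skin Mask, darkened at Tattoos, then blended in by Transition, excluding metallic areas unless isHair is set.
# Inverted Transition and Flame Detail index a gradient texture added to emissive, masked by Flame Edge and skin minus tattoos.
# The IsBody switch is left unconnected to the output.
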MaterialFunction /BRCosmetics/Characters/Player/Male/Medium/Bodies/M_MED_Headset_Fantastical/Materials/MF_Headset_PurpleSkin
graph {nodes [
  n1 = GetMaterialAttributes(None=FunctionInput() {inputType=FunctionInput_MaterialAttributes}) {attributeGetTypes [#1, #2, #3]}
  n2 = FunctionInput(Preview=Constant() {r=1}) {inputName=Transition, inputType=FunctionInput_Scalar}
  n3 = OneMinus(None=@n2)
  n4 = ComponentMask(None=FunctionInput() {inputName=Flame Detail}) {r=true}
  n5 = Saturate(None=Multiply(A=ScalarParameter() {defaultValue=1.25, parameterName=Flame Offset}, B=Multiply(A=@n3, B=@n4)))
  n6 = Add(
   A=@n1.EmissiveColor
   B=Multiply(
    A=ScalarParameter() {defaultValue=3, parameterName=Edge Brightness}
    B=TextureSample(UVs=@n5) {texture=/BRCosmetics/Characters/Player/Male/Medium/Bodies/M_MED_Headset_Fantastical/Textures/HeadsetGradient_Grad}.RGB
   )
  )
  n7 = Multiply(
   A=ScalarParameter() {defaultValue=1, parameterName=Flame Edge Mask Brightness}
   B=FunctionInput() {inputName=Flame Edge, inputType=FunctionInput_Scalar}
  )
  n8 = FunctionInput() {inputName=Skin Mask, inputType=FunctionInput_Scalar}
  n9 = FunctionInput(Preview=TextureSample() {samplerType=SAMPLERTYPE_LinearColor, texture=/Game/Athena/Apollo/Water/Textures/02_Rounded/03_WaterFoam_Veroni_base_H}.RGB) {inputName=Tattoos, inputType=FunctionInput_Scalar}
  n10 = Subtract(A=@n8, B=@n9)
  n11 = Saturate(None=Multiply(A=@n4, B=Multiply(A=@n7, B=@n10)))
  n12 = Constant() {r=0.9}
  n13 = FunctionInput() {inputName=Gmap, inputType=FunctionInput_Scalar}
  n14 = LinearInterpolate(
   A=Multiply(A=VectorParameter() {parameterName=Other stuff tinted, defaultValue=<rgba 0.8659 0.8385 1.4 1>}.RGB, B=@n13)
   B=Multiply(A=VectorParameter() {parameterName=PurpleSkin, defaultValue=<rgba 0.6114 0.5813 1.2 1>}.RGB, B=@n13)
   Alpha=@n8
  )
  n15 = LinearInterpolate(
   A=Multiply(A=Desaturation(None=@n1.BaseColor, Fraction=@n12), B=@n14)
   B=Constant3Vector() {constant=<rgba 0.0206 0.0074 0.026 1>}
   Alpha=Saturate(None=@n9)
  )
  n16 = Round(None=@n1.Metallic)
  n17 = Saturate(None=StaticSwitchParameter(True=@n2, False=Subtract(A=@n2, B=@n16)) {parameterName=isHair})
  FunctionOutput(None=SetMaterialAttributes!(
    MaterialAttributes=@n1.MaterialAttributes
    EmissiveColor=LinearInterpolate(A=@n1.EmissiveColor, B=@n6, Alpha=@n11)
    BaseColor=LinearInterpolate(A=@n1.BaseColor, B=@n15, Alpha=@n17)
   ) {attributeSetTypes [#2, #1]})
  StaticSwitchParameter(False=@n14) {parameterName=IsBody}
 ]}
