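# Parent material for the Fantastical Headset character body, with a flame-edged reveal that drives purple-skin and moving-tattoo effects.
# A sphere mask in pre-skinned local space grows with FlameRevealAmount or an override, eroded by panning triplanar fire noise to form transition masks.
# Those masks feed the skin and tattoo functions between the shared character PreFX and PostFX functions.
# Rounded metallic masks a metals LUT and varies the noise panning.
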Material /BRCosmetics/Characters/Player/Male/Medium/Bodies/M_MED_Headset_Fantastical/Materials/M_Headset_Fantastical_Parent
graph {
 properties {
  bUseMaterialAttributes=true
  bUsedWithMeshDeformer=true
  bUsedWithMorphTargets=true
  bUsedWithSkeletalMesh=true
  pixelDepthOffsetMode=PDOM_Legacy
  usedShadingModels=Default Lit
 }
 nodes [
  n1 = MaterialFunctionCall() {materialFunction=&BaseFN_CharacterFunction_PreFX}
  n2 = NamedRerouteUsage()
  n3 = NamedRerouteUsage()
  n4 = NamedRerouteUsage()
  n5 = TextureObject() {samplerType=SAMPLERTYPE_Masks, texture=/BRCosmetics/Characters/Player/Male/Medium/Bodies/M_MED_Headset_Fantastical/Textures/HeadsetFire_M}
  n6 = Constant() {r=-0.125}
  n7 = Constant() {r=0.1}
  n8 = Constant()
  n9 = Constant() {r=0.25}
  n10 = Round(None=GetMaterialAttributes(None=@n1.Result) {attributeGetTypes [#1]}.Metallic)
  n11 = MaterialFunctionCall!(
   Export Float 4=StaticBool() {value=true}
   TextureObject=@n5
   TextureSize=ScalarParameter() {defaultValue=45, parameterName=Panner Size}
   WorldPosition=@n1.Pre skinned Local Position
   World Space Normal
   ProjectionTransitionContrast=Constant()
   Pan_YZ=AppendVector(A=Constant() {r=-0.1}, B=Constant() {r=0.05})
   Pan_XZ=LinearInterpolate(A=AppendVector(A=@n6, B=@n7), B=AppendVector(A=@n8, B=@n9), Alpha=@n10)
   Pan_XY=AppendVector(A=Constant() {r=-0.2}, B=Constant())
  ) {materialFunction=/Game/Weapons/WeaponSkins/Wraps/MaterialFunctions/WorldAlignedTexture_ComplexPan}
  n12 = MaterialFunctionCall(Float3=@n11.XYZ Texture) {materialFunction=&BreakOutFloat3Components}
  n13 = Max(
   A=ScalarParameter() {group=Flames, parameterName=FlameRevealAmount}
   B=LinearInterpolate(Alpha=ScalarParameter() {parameterName=FlameRevealOverride}) {constB=1.25}
  )
  n14 = Multiply(A=OneMinus(None=NamedRerouteUsage())) {constB=0.15}
  n15 = SphereMask(
   A=NamedRerouteUsage()
   B=VectorParameter() {group=HeadsetFire, parameterName=Position, defaultValue=<rgba 28.6266 0 100 1>}.RGB
   Radius=LinearInterpolate(B=ScalarParameter() {defaultValue=120, group=HeadsetFire, parameterName=FadeRadius}, Alpha=Subtract(A=@n13, B=@n14)) {constB=0}
   Hardness=ScalarParameter() {defaultValue=0.66, parameterName=Hardness}
  ) {attenuationRadius=50, hardnessPercent=0}
  n16 = Saturate(None=@n15)
  n17 = SphereMask(A=@n16, B=Constant() {r=-0.02}, Radius=ScalarParameter() {defaultValue=0.98, parameterName=Radius}) {hardnessPercent=0.2}
  n18 = Multiply(A=Multiply(A=OneMinus(None=@n12.B), B=@n17), B=@n3)
  n19 = TextureObjectParameter() {
   parameterName=SkinFX
   samplerType=SAMPLERTYPE_Masks
   texture=/BRCosmetics/Characters/Player/Male/Medium/Bodies/M_MED_Headset_Fantastical/Textures/T_M_Headset_Fantastical_Body_FX
  }
  n20 = MaterialFunctionCall!(
   SkinFX=@n19
   transition Amount=NamedRerouteUsage()
   Distortion Noise=NamedRerouteUsage()
   Character Skin=NamedRerouteUsage()
   Material=@n1.Result
  ) {materialFunction=&/MF_Headset_MovingTattoos}
  MaterialFunctionCall(
   AO=@n1.AO
   Distance Blend=@n1.AO
   Fresnel=@n1.AO
   Gmap=@n1.AO
   Pre Skinned Local Position=@n1.AO
   Material=MaterialFunctionCall(
    Mask=@n10
    Materials=MaterialFunctionCall!(Skin Mask=@n2, Transition=@n3, Gmap=@n4, Flame Edge=@n18, Tattoos=@n20.Flame Mask, Flame Detail=@n11.XYZ Texture, In=@n20.Flame Mask) {materialFunction=&/MF_Headset_PurpleSkin}.Result
   ) {materialFunction=&MF_MetalsLUT}.Result
  ) {materialFunction=&BaseFN_CharacterFunction_PostFX}
  NamedRerouteDeclaration(None=@n1.Pre skinned Local Position) {name=Pre Skinned, nodeColor=<rgba 1 0.5156 0 1>}
  NamedRerouteDeclaration(None=@n1.Skin) {name=Skin Mask, nodeColor=<rgba 1 0 0.2578 1>}
  NamedRerouteDeclaration(None=@n1.Gmap) {name=Gmap, nodeColor=<rgba 0.5 0 1 1>}
  n21 = Power(Base=@n12.G)
  n22 = Multiply(A=Saturate(None=Multiply(A=@n17, B=@n21)), B=ScalarParameter() {defaultValue=5, parameterName=Transition Flame Intensity})
  n23 = Subtract(A=@n16, B=Saturate(None=@n22))
  NamedRerouteDeclaration(None=Saturate(None=Multiply(A=@n23, B=@n15))) {name=Headset Transition, nodeColor=<rgba 1 0.4922 0 1>}
  NamedRerouteDeclaration(None=@n11.XYZ Texture) {name=Distortion Noise, nodeColor=<rgba 1 0 0.0703 1>}
  NamedRerouteDeclaration(None=@n10) {name=Rounded Metal mask, nodeColor=<rgba 0 1 0.4844 1>}
  n24 = Subtract(A=@n16, B=Saturate(None=Multiply(A=@n22) {constB=0.1}))
  n25 = Multiply(A=@n15) {constB=1.3}
  NamedRerouteDeclaration(None=Saturate(None=Multiply(A=@n24, B=@n25))) {name=Headset Transition Tattoos, nodeColor=<rgba 1 0.4922 0 1>}
  NamedRerouteUsage()
 ]
}
