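# Physics asset for the male medium Fantastical Headset player body, covering a left zipper dangle bone chain and the left thigh.
# Three capsule bodies form the chain: a large thigh capsule, a small zipper master capsule and the zipper segment.
# The master rotates freely on all angular axes relative to the thigh, while the segment has narrow swing limits and a zero twist limit relative to the master.
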PhysicsAsset /Game/Characters/Player/Male/Medium/Bodies/M_MED_Headset_Fantastical/Meshes/M_MED_Headset_Fantastical_Physics
bodies {
 dyn_zipper_l_01 [{
   shapeType=Capsule
   length=2.3169
   radius=0.8691
   shapeName=dyn_zipper_l_01_capsule
   center=<2.3368 -0.0342 0.0094>
   rotation=<rot 0 89.8826 -89.3528>
  }]
 dyn_zipper_l_MSTR [{shapeType=Capsule, length=0.51, radius=0.5, shapeName=dyn_zipper_l_MSTR_capsule, rotation=<rot 90 0 0>}]
 thigh_l [{shapeType=Capsule, length=21.4982, radius=8.8987, shapeName=thigh_l_capsule, center=<-15.0829 0.446 0.1491>, rotation=<rot 90 0 0>}]
}
constraints [
 {bone1Name=dyn_zipper_l_01, bone2Name=dyn_zipper_l_MSTR, swing1LimitDegrees=5, swing2LimitDegrees=20, twistLimitDegrees=0}
 {
  bone1Name=dyn_zipper_l_MSTR
  bone2Name=thigh_l
  swing1LimitDegrees=45
  swing1Motion=Free
  swing2LimitDegrees=45
  swing2Motion=Free
  twistLimitDegrees=45
  twistMotion=Free
 }
]
