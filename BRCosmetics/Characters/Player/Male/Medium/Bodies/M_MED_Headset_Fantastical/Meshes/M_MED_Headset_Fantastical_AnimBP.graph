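# Post-process style AnimBP for the Fantastical medium male body, copying the parent mesh pose and adding secondary motion.
# Bone constraints drive the left shoulder pad, back star and left zipper, with deform and shoulder-pad Control Rigs between and spring-based Anim Dynamics on the zipper.
# Additive shoulder-pad offsets apply for an open parachute or pickaxe, the latter flagged each update from the weapon animation type.
# A RigidBody node sits unconnected.
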AnimBlueprint /BRCosmetics/Characters/Player/Male/Medium/Bodies/M_MED_Headset_Fantastical/Meshes/M_MED_Headset_Fantastical_AnimBP
target_skeleton=&/M_MED_Headset_Fantastical_Skeleton
graphs [
 AnimGraph {nodes [
   RigidBody() {
    overridePhysicsAsset=/Game/Characters/Player/Male/Medium/Bodies/M_MED_Headset_Fantastical/Meshes/M_MED_Headset_Fantastical_Physics
    simulationSpace=BaseBoneSpace
    alphaBoolBlend {blendOption=Linear}
    baseBoneRef {boneName=thigh_l}
    componentLinearAccScale=<0.5 0.5 0.5>
    componentLinearVelScale=<0.5 0.5 0.5>
   }
   n1 = "Constraint - shoulderPad_l_01 (PPPP)"!(
    ComponentPose=Local To Component(LocalPose=Copy Pose From Mesh() {bUseAttachedParent=true})
    Alpha=1
    ConstraintWeights_0=1
    ConstraintWeights_1:float=0
    ConstraintWeights_2:float=0
    ConstraintWeights_3:float=0
   ) -> (Pose) {
    alphaBoolBlend {blendOption=Linear}
    boneToModify {boneName=shoulderPad_l_01}
    constraintSetup [
     {targetBone {boneName=upperarm_twist_01_l}}
     {targetBone {boneName=deform_upperarm_up_l}}
     {targetBone {boneName=deform_upperarm_bk_l}}
     {targetBone {boneName=clavicle_l}}
    ]
    constraintWeights [1,1,1,1]
   }
   n2 = "Constraint - dyn_backStar_01 (PPP)"!(ComponentPose=@n1, Alpha=1, ConstraintWeights_0=0.4, ConstraintWeights_1:float=0.4, ConstraintWeights_2:float=0.2) -> (Pose) {
    alphaBoolBlend {blendOption=Linear}
    boneToModify {boneName=dyn_backStar_01}
    constraintSetup [{targetBone {boneName=clavicle_l}}, {targetBone {boneName=clavicle_r}}, {targetBone {boneName=spine_04}}]
    constraintWeights [1,1,1]
   }
   n3 = "Transform (Modify) Bone - Bone: dyn_backStar_01"(ComponentPose=@n2, Alpha=1, Translation="0,-1,0", Rotation="0,0,0", Scale="1,1,1") {
    translationMode=BMM_Additive
    translationSpace=BCS_BoneSpace
    alphaBoolBlend {blendOption=Linear}
    boneToModify {boneName=dyn_backStar_01}
    translation {y=-1}
   }
   n4 = Control Rig(Source=Component To Local(ComponentPose=@n3)) {
    alphaBoolBlend {blendOption=Linear}
    controlRigAssetReference {blueprintRigClass=&/M_MED_DeformRig_Headset_Fantastical_ControlRig.M_MED_DeformRig_Headset_Fantastical_ControlRig_C}
   }
   n5 = Control Rig(Source=@n4) {
    alphaBoolBlend {blendOption=Linear}
    controlRigAssetReference {blueprintRigClass=&/M_MED_Headset_Fantastical_ShoulderPad_CtrlRig.M_MED_Headset_Fantastical_ShoulderPad_CtrlRig_C}
   }
   n6 = "Constraint - dyn_zipper_l_01 (PPPPP)"!(
    ComponentPose=Local To Component(LocalPose=@n5)
    Alpha=1
    ConstraintWeights_0=0.748
    ConstraintWeights_1:float=0
    ConstraintWeights_2:float=0.065
    ConstraintWeights_3:float=0.091
    ConstraintWeights_4:float=0.096
   ) -> (Pose) {
    alphaBoolBlend {blendOption=Linear}
    boneToModify {boneName=dyn_zipper_l_01}
    constraintSetup [
     {targetBone {boneName=thigh_twist_01_l}}
     {targetBone {boneName=deform_glute_l}}
     {targetBone {boneName=thigh_l}}
     {targetBone {boneName=pelvis}}
     {targetBone {boneName=deform_groin_l}}
    ]
    constraintWeights [1,1,1,1,1]
   }
   n7 = Anim Dynamics - Bone: dyn_zipper_l_01(ComponentPose=@n6, Alpha=1, ExternalForce="0,0,0") {
    angularDampingOverride=1
    angularSpringConstant=20000
    bAngularSpring=true
    bOverrideAngularDamping=true
    bUseGravityOverride=true
    gravityScale=0.35
    simulationSpace=BoneRelative
    alphaBoolBlend {blendOption=Linear}
    boundBone {boneName=dyn_zipper_l_01}
    componentAppliedLinearAccClamp=<6000 6000 6000>
    componentLinearAccScale=<1 1 1>
    componentLinearVelScale=<1 1 1>
    gravityOverride {z=-980}
    physicsBodyDefinitions [{
      boundBone {boneName=dyn_zipper_l_01}
      boxExtents=<10 10 10>
      constraintSetup {angularLimitsMax=<0 0 10>, angularLimitsMin=<0 -30 -10>, angularTarget=<0 0 0>}
      localJointOffset=<10 0 0>
     }]
    relativeSpaceBone {boneName=thigh_l}
   }
   n8 = "Transform (Modify) Bone - Bone: shoulderPad_l_01"!(
    ComponentPose=@n7
    bAlphaBoolEnabled:bool=Get bIsParachuteOpened()
    Translation="2.667,-5.670,3.060"
    Rotation="-11.964,19.285,-63.401"
    Scale="1,1,1"
   ) -> (Pose) {
    alphaInputType=Bool
    rotationMode=BMM_Additive
    rotationSpace=BCS_BoneSpace
    translationMode=BMM_Additive
    translationSpace=BCS_BoneSpace
    alphaBoolBlend {blendOption=Linear}
    boneToModify {boneName=shoulderPad_l_01}
    rotation=<rot -11.964 19.285 -63.401>
    translation=<2.667 -5.67 3.06>
   }
   n9 = "Transform (Modify) Bone - Bone: shoulderPad_l_01"(ComponentPose=@n8, bAlphaBoolEnabled=Get has pickaxe() -> (has pickaxe:bool), Translation="0,1,0.75", Rotation="0,0,3", Scale="1,1,1") {
    alphaInputType=Bool
    rotationMode=BMM_Additive
    rotationSpace=BCS_BoneSpace
    translationMode=BMM_Additive
    translationSpace=BCS_BoneSpace
    alphaBoolBlend {blendOption=Linear}
    boneToModify {boneName=shoulderPad_l_01}
    rotation {roll=3}
    translation {y=1, z=0.75}
   }
   Output Pose(Result=Component To Local(ComponentPose=@n9)) {layerGroup=None}
  ]}
 EventGraph {nodes [
   Event BlueprintUpdateAnimation()
   > n1 = Cast To FortPlayerAnimInstance(Object=GetAnimInstance(self=GetOwnerSkeletalMeshComponent()))
   > Set has pickaxe(has pickaxe:bool=false) -> (Output_Get:bool)
   > Switch on EFortWeaponCoreAnimation!(execute, Selection:EFortWeaponCoreAnimation Enum=Get WeaponCoreAnim(self=@n1)) -> (
    Melee:Exec
    Pistol:Exec
    Shotgun:Exec
    PaperBlueprint:Exec
    Rifle:Exec
    MeleeOneHand:Exec
    MachinePistol:Exec
    RocketLauncher:Exec
    GrenadeLauncher:Exec
    GoingCommando:Exec
    AssaultRifle:Exec
    TacticalShotgun:Exec
    SniperRifle:Exec
    TrapPlacement:Exec
    ShoulderLauncher:Exec
    AbilityDecoTool:Exec
    Crossbow:Exec
    C4:Exec
    RemoteControl:Exec
    DualWield:Exec
    AR_BullPup:Exec
    AR_ForwardGrip:Exec
    MedPackPaddles:Exec
    SMG_P90:Exec
    AR_DrumGun:Exec
    Consumable_Small:Exec
    Consumable_Large:Exec
    Balloon:Exec
    MountedTurret:Exec
    CreativeTool:Exec
    ExplosiveBow:Exec
    AshtonIndigo:Exec
    AshtonChicago:Exec
    MeleeDualWield:Exec
    Unarmed:Exec
   )
   > Set has pickaxe(has pickaxe=true)
  ]}
]
variables [
 has pickaxe {type=boolean}
 is weapon {type=boolean}
]
